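# Translucent deferred decal material for the CornerWeek emote, building an animated radial ground effect from scrolling noise.
# Polar UVs drive panned noise and smoke textures, shaped by center sphere masks into decal-colored glow and a time-of-day tinted layer.
# A ripple mask, radial gradient and decal lifetime form opacity, while two animated normal maps blend radially and pass through FlattenNormal.
# Color samples on polar UVs take mip derivatives from plain texture coordinates.
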Material /BRCosmetics/Effects/Fort_Effects/Emotes/CornerWeek/M_Emote_CornerWeek_Decal
graph {
 properties {
  bEnableResponsiveAA=true
  bOutputTranslucentVelocity=true
  bUsedWithNiagaraRibbons=true
  bUsedWithNiagaraSprites=true
  bUsedWithParticleSprites=true
  blendMode=BLEND_Translucent
  materialDomain=MD_DeferredDecal
  pixelDepthOffsetMode=PDOM_Legacy
  shadingModel=MSM_Unlit
  twoSided=true
  usedShadingModels=Unlit
 }
 nodes [
  Constant()
  n1 = MaterialFunctionCall() {materialFunction=&VectorToRadialValue}
  n2 = Multiply(A=@n1.Radial Coordinates, B=Constant2Vector() {g=1.5, r=3})
  n3 = Constant2Vector() {g=1, r=2}
  n4 = Time()
  n5 = ScalarParameter() {defaultValue=1, parameterName=TimeSpeed}
  n6 = TextureSampleParameter2D(UVs=Panner(Coordinate=Multiply(A=@n1.Radial Coordinates, B=@n3), Time=Multiply(A=@n4, B=@n5)) {speedY=0.15}) {
   group=Distortion
   parameterName=NoiseTexture
   samplerType=SAMPLERTYPE_Normal
   sortPriority=0
   texture=&T_Weap_Distort_N_02
  }
  n7 = Multiply(A=ComponentMask(None=@n6.RGB) {g=true, r=true}, B=ScalarParameter() {defaultValue=0.15, parameterName=Noise_intensity})
  n8 = Time()
  n9 = ScalarParameter() {defaultValue=1, parameterName=TimeSpeed}
  n10 = TextureCoordinate()
  n11 = TextureSample!(
   UVs=Panner(Coordinate=Add(A=@n2, B=@n7), Time=Multiply(A=@n8, B=@n9)) {speedY=0.5}
   Tex
   "DDX(UVs)"=DDX(Value=@n10)
   "DDY(UVs)"=DDY(Value=@n10)
   Apply View MipBias
  ) {mipValueMode=TMVM_Derivative, texture=/Game/Effects/Fort_Effects/Textures/Tile/T_Noise_03}
  n12 = Panner(
   Coordinate=Multiply(A=@n1.Radial Coordinates, B=Constant2Vector() {g=1, r=3})
   Time=Multiply(A=Time(), B=ScalarParameter() {defaultValue=1, parameterName=TimeSpeed})
  ) {speedY=0.2}
  n13 = TextureCoordinate()
  n14 = DDX(Value=@n13)
  n15 = DDY(Value=@n13)
  n16 = LinearInterpolate(
   A=Multiply(A=LinearInterpolate(Alpha=@n11.RGB)) {constB=2}
   B=LinearInterpolate(Alpha=TextureSample(UVs=@n12, "DDX(UVs)"=@n14, "DDY(UVs)"=@n15) {mipValueMode=TMVM_Derivative, texture=&T_Inky_Smoke_Tile}.R)
  )
  n17 = SphereMask(A=TextureCoordinate(), B=Constant2Vector() {g=0.5, r=0.5}, Radius=Constant() {r=0.2}, Hardness=Constant())
  n18 = Saturate(None=Power(Base=OneMinus(None=@n17)))
  MaterialFunctionCall(Emissive=Multiply(A=DecalColor().RGB, B=Multiply(A=@n16, B=@n18))) {materialFunction=&MF_FX_EmissiveAdjustments}
  n19 = TextureSampleParameter2D(UVs=MaterialFunctionCall(Texture Scale=Constant() {r=0.6}) {materialFunction=&ScaleUVsByCenter}.UVs) {
   parameterName=Mask
   samplerSource=SSM_Clamp_WorldGroupSettings
   samplerType=SAMPLERTYPE_LinearColor
   texture=/Game/Effects/Fort_Effects/Effects/Water/Textures/WaterDecalRippleMask
  }
  n20 = ScalarParameter() {defaultValue=0.5, parameterName="Radius (S)"}
  n21 = ScalarParameter() {defaultValue=1, parameterName="Density (S)"}
  n22 = StaticBool()
  n23 = Multiply(
   A=@n19.R
   B=Multiply(A=MaterialFunctionCall(Radius=@n20, Density=@n21, Invert Density=@n22) {materialFunction=&RadialGradientExponential}.RadialGradientExponential) {constB=2}
  ) {constB=2}
  n24 = ScalarParameter() {defaultValue=1.5, parameterName=Mask_Mul}
  Multiply(
   A=Power(Base=Multiply(A=@n23, B=@n24) {constB=1.5}, Exp=ScalarParameter() {defaultValue=1, parameterName=Mask_Power}) {constExponent=1}
   B=DecalLifetimeOpacity().Opacity
  )
  n25 = VectorParameter() {parameterName=BottomColor, defaultValue=<rgba 0.229 0.229 1 1>}
  n26 = Constant2Vector() {g=0.5, r=2}
  n27 = Time()
  n28 = ScalarParameter() {defaultValue=1, parameterName=TimeSpeed}
  n29 = TextureCoordinate()
  n30 = TextureSample(
   UVs=Panner(Coordinate=Multiply(A=@n1.Radial Coordinates, B=@n26), Time=Multiply(A=@n27, B=@n28)) {speedY=0.2}
   "DDX(UVs)"=DDX(Value=@n29)
   "DDY(UVs)"=DDY(Value=@n29)
  ) {mipValueMode=TMVM_Derivative, texture=/BRCosmetics/Weapons/FORT_Melee/Pickaxe_CoatCheck/FX/T_Pickaxe_CoatCheck_Noise}
  n31 = OneMinus(None=SphereMask(A=TextureCoordinate(), B=Constant2Vector() {g=0.5, r=0.5}, Radius=Constant() {r=0.2}, Hardness=Constant()))
  n32 = SphereMask(A=TextureCoordinate(), B=Constant2Vector() {g=0.5, r=0.5}, Radius=Constant() {r=0.35}, Hardness=Constant())
  n33 = Power(Base=Multiply(A=Multiply(A=@n31, B=@n32)) {constB=1.5})
  Multiply(
   A=MaterialFunctionCall(
    Color_Morning=@n25.RGB
    Color_Day=@n25.RGB
    Color_Evening=@n25.RGB
    Color_Night=VectorParameter() {parameterName=BottomColorNight, defaultValue=<rgba 0.1866 0.1866 0.4479 1>}.RGB
   ) {materialFunction=&Color_Time_of_Day}.Result
   B=Multiply(A=Power(Base=@n30.R) {constExponent=1}, B=Saturate(None=@n33))
  )
  n34 = Constant() {r=0.6}
  n35 = TextureSampleParameter2D(UVs=Panner(Coordinate=TextureCoordinate() {uTiling=3, vTiling=3}) {speedX=0.1, speedY=0.3}) {
   group=Distortion
   parameterName=NormalNoiseTexture
   samplerType=SAMPLERTYPE_Normal
   sortPriority=0
   texture=&T_Weap_Distort_N_02
  }
  n36 = ComponentMask(None=@n35.RGB) {g=true, r=true}
  n37 = ScalarParameter() {parameterName=NormalNoise_intensity}
  n38 = TextureSample!(
   UVs=Add(
    A=MaterialFunctionCall(Texture Scale=@n34) {materialFunction=&ScaleUVsByCenter}.UVs
    B=Multiply(A=@n36, B=@n37)
   )
   Tex
   Apply View MipBias
  ) {
   samplerSource=SSM_Clamp_WorldGroupSettings
   samplerType=SAMPLERTYPE_Normal
   texture=&/T_Emote_CornerWeek_Normal02
  }
  n39 = MaterialFunctionCall() {materialFunction=&VectorToRadialValue}
  n40 = Constant2Vector() {g=0.4, r=1}
  n41 = Time()
  n42 = ScalarParameter() {defaultValue=1, parameterName=TimeSpeed}
  n43 = TextureSample!(
   UVs=Panner(Coordinate=Multiply(A=@n39.Radial Coordinates, B=@n40), Time=Multiply(A=@n41, B=@n42)) {speedY=0.05}
   Tex
   Level
   Apply View MipBias
  ) {
   mipValueMode=TMVM_MipLevel
   samplerType=SAMPLERTYPE_Normal
   texture=&/T_Emote_CornerWeek_Normal
  }
  n44 = SphereMask(A=TextureCoordinate(), B=Constant2Vector() {g=0.5, r=0.5}, Radius=Constant() {r=0.1}, Hardness=Constant())
  n45 = Saturate(None=Power(Base=OneMinus(None=@n44)) {constExponent=2.5})
  Reroute(None=MaterialFunctionCall(Normal=LinearInterpolate(A=@n38.RGB, B=@n43.RGB, Alpha=@n45), Flatness=ScalarParameter() {defaultValue=-0.3, parameterName="Flatness (S)"}) {materialFunction=&FlattenNormal}.Result)
  ScalarParameter() {parameterName=BaseColor}
  ScalarParameter() {defaultValue=0.5, parameterName=Specular}
  ScalarParameter() {parameterName=Metallic}
  ScalarParameter() {defaultValue=1, parameterName=Roughness}
  VectorParameter() {parameterName=v, defaultValue=<rgba 0.05 0.004 0.0021 1>}
 ]
}
