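# Translucent, unlit, two-sided noise material for Niagara mesh particles in the CornerWeek emote effects.
# Emissive pans a noise-distorted swirl texture, tinted by particle color.
# Opacity multiplies a noise-warped panning mask by particle alpha and dynamic-parameter top/bottom UV masks, with optional dissolve and linear mask, then view-angle and depth fades.
# Vertex offset combines panning-texture wobble along normals with optional local-axis displacement driven by a second dynamic parameter.
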Material /BRCosmetics/Effects/Fort_Effects/Emotes/CornerWeek/M_Emote_CornerWeek_Noise02
graph {
 properties {
  bAutomaticallySetUsageInEditor=false
  bUsedWithNiagaraMeshParticles=true
  blendMode=BLEND_Translucent
  shadingModel=MSM_Unlit
  twoSided=true
  usedShadingModels=Unlit
 }
 nodes [
  n1 = Transform(None=Constant3Vector() {constant=<rgba 1 0 0 0>}) {transformSourceType=TRANSFORMSOURCE_Local}
  n2 = OneMinus(None=ComponentMask(None=TextureCoordinate()) {g=true})
  n3 = DynamicParameter() {parameterIndex=1, paramNames [X WPO_Intensity, Y WPO_Intensity, Z WPO_Intensity, Param4]}
  n4 = Transform(None=Constant3Vector() {constant=<rgba 0 1 0 0>}) {transformSourceType=TRANSFORMSOURCE_Local}
  n5 = Add(A=Multiply(A=Multiply(A=@n1, B=@n2), B=@n3.X WPO_Intensity) {constB=10}, B=Multiply(A=@n3.Y WPO_Intensity, B=Multiply(A=@n4, B=@n2)))
  n6 = Constant3Vector() {constant=<rgba 0 0 1 0>}
  n7 = Multiply(A=@n3.Z WPO_Intensity, B=Multiply(A=Transform(None=@n6) {transformSourceType=TRANSFORMSOURCE_Local}, B=@n2))
  n8 = TextureCoordinate() {uTiling=0, vTiling=2}
  n9 = ScalarParameter() {defaultValue=1, parameterName=WPOTex_UVTIling}
  n10 = Multiply(A=Time(), B=ScalarParameter() {defaultValue=1, parameterName=WPO_Speed})
  n11 = TextureSampleParameter2D(UVs=Panner(Coordinate=Multiply(A=@n8, B=@n9), Time=@n10) {speedY=1}) {parameterName=WPOTexture, samplerType=SAMPLERTYPE_Normal, texture=&T_Weap_Distort_N_02}
  n12 = LinearInterpolate(Alpha=@n11.R) {constB=1.5}
  n13 = TextureCoordinate() {uTiling=0}
  n14 = TextureSampleParameter2D(UVs=Panner(Coordinate=Multiply(A=@n9, B=@n13), Time=@n10) {speedY=0.8}) {parameterName=WPOTexture, samplerType=SAMPLERTYPE_Normal, texture=&T_Weap_Distort_N_02}
  n15 = LinearInterpolate(Alpha=@n14.G) {constB=1.5}
  n16 = Constant() {r=1}
  n17 = OneMinus(None=ComponentMask(None=TextureCoordinate()) {g=true})
  n18 = Multiply(A=VertexNormalWS(), B=Multiply(A=Add(A=@n12, B=@n15), B=LinearInterpolate(A=@n16, Alpha=@n17) {constA=0.05}))
  n19 = Multiply(A=Transform(None=Constant3Vector() {constant=<rgba 1 0 0 0>}) {transformSourceType=TRANSFORMSOURCE_Local}, B=@n18)
  n20 = ScalarParameter() {defaultValue=5, parameterName=WPO_Intensity}
  n21 = Multiply(A=Transform(None=Constant3Vector() {constant=<rgba 0 1 0 0>}) {transformSourceType=TRANSFORMSOURCE_Local}, B=@n18)
  Add(
   A=LinearInterpolate(A=Add(A=@n5, B=@n7), Alpha=ScalarParameter() {parameterName=Use_XYZ_WPO}) {constB=0}
   B=Add(A=Multiply(A=@n19, B=@n20), B=Multiply(A=@n21, B=@n20))
  )
  n22 = ParticleColor()
  n23 = TextureCoordinate()
  n24 = AppendVector(
   A=ScalarParameter() {defaultValue=1, parameterName=MaskTextureTilingU}
   B=ScalarParameter() {defaultValue=1, parameterName=MaskTextureTilingV}
  )
  n25 = Time()
  n26 = ScalarParameter() {defaultValue=1, parameterName=MaskNoise_Speed}
  n27 = TextureSample(UVs=Panner(Coordinate=TextureCoordinate() {uTiling=1.5, vTiling=1.5}, Time=Multiply(A=@n25, B=@n26)) {speedY=0.2}) {samplerType=SAMPLERTYPE_Grayscale, texture=&LowResBlurredNoise}
  n28 = ScalarParameter() {defaultValue=0.1, parameterName=MaskTexture_NoiseIntensity}
  n29 = Panner(
   Coordinate=Add(A=Multiply(A=@n23, B=@n24), B=Multiply(A=@n27.R, B=@n28))
   Time=Multiply(A=Time(), B=ScalarParameter() {defaultValue=1, parameterName=MaskTexture_Speed})
   Speed=AppendVector(A=ScalarParameter() {parameterName=MaskTextureSpeedX}, B=ScalarParameter() {defaultValue=0.4, parameterName=MaskTextureSpeedY})
  ) {speedX=0.1, speedY=0.4}
  n30 = Multiply(
   A=@n22.A
   B=ChannelMaskParameter(None=TextureSampleParameter2D(UVs=@n29) {parameterName=MaskTexture, texture=/Game/Weapons/WeaponSkins/Textures/T_Swirly_Noise_highContrast}.RGBA) {parameterName=MaskTexture_RGB}.R
  )
  n31 = NamedRerouteUsage()
  n32 = ComponentMask(None=TextureCoordinate()) {g=true}
  n33 = DynamicParameter() {paramNames [TopMask,BottomMask,Param3,Param4]}
  n34 = ScalarParameter() {defaultValue=1, parameterName=Top_Mask_Multiply}
  n35 = Power(Base=Multiply(A=@n32, B=Multiply(A=@n33.TopMask, B=@n34)) {constB=6}, Exp=ScalarParameter() {defaultValue=3, parameterName=Top_Mask_Power}) {constExponent=3}
  n36 = OneMinus(None=@n32)
  n37 = DynamicParameter() {paramNames [TopMask,BottomMask,Param3,Param4]}
  n38 = ScalarParameter() {defaultValue=1, parameterName=B_Mask_Multiply}
  n39 = Power(
   Base=Multiply(A=Multiply(A=@n36), B=Multiply(A=@n37.BottomMask, B=@n38)) {constB=6}
   Exp=ScalarParameter() {defaultValue=3, parameterName=B_Mask_Power}
  ) {constExponent=3}
  n40 = Multiply(
   A=LinearInterpolate(A=@n30, B=Multiply(A=@n30, B=@n31), Alpha=ScalarParameter() {parameterName=Use_Dissolve})
   B=Multiply(A=Saturate(None=@n35), B=Saturate(None=@n39))
  )
  n41 = ComponentMask(None=TextureCoordinate()) {r=true}
  n42 = Multiply(
   A=Power(Base=Sine(None=@n41) {period=2}, Exp=ScalarParameter() {defaultValue=1, parameterName=LinearMask_Power})
   B=ScalarParameter() {defaultValue=2, parameterName=LinearMask_Mul}
  ) {constB=7}
  n43 = LinearInterpolate(A=@n40, B=Multiply(A=@n40, B=Saturate(None=@n42)), Alpha=ScalarParameter() {parameterName=Use_LinearMask})
  n44 = ScalarParameter() {defaultValue=1, parameterName=All_Dark}
  n45 = ScalarParameter() {defaultValue=1, group=Opacity, parameterName=MeshFade}
  DepthFade(
   Opacity=Multiply(
    A=Multiply(A=@n43, B=@n44)
    B=MaterialFunctionCall("FadeSensitivity (optional)"=@n45) {materialFunction=&FadeBasedOnViewAngle}.Result
   )
   FadeDistance=ScalarParameter() {defaultValue=5, parameterName=DepthFade}
  )
  n46 = TextureCoordinate()
  n47 = AppendVector(A=ScalarParameter() {defaultValue=1, parameterName=LightTextureTilingU}, B=Constant() {r=1})
  n48 = TextureSampleParameter2D(UVs=Panner(Coordinate=TextureCoordinate()) {speedY=0.3}) {
   group=Distortion
   parameterName=NoiseTexture
   samplerType=SAMPLERTYPE_Normal
   sortPriority=0
   texture=&T_Weap_Distort_N_02
  }
  n49 = Multiply(A=ComponentMask(None=@n48.RGB) {g=true, r=true}, B=ScalarParameter() {defaultValue=0.15, parameterName=Noise_intensity})
  n50 = TextureCoordinate()
  n51 = Constant2Vector() {g=-0.03}
  n52 = OneMinus(None=ComponentMask(None=Add(A=@n50, B=@n51)) {g=true})
  n53 = Panner(
   Coordinate=Add(A=Multiply(A=@n46, B=@n47), B=Multiply(A=@n49, B=@n52))
   Time=Multiply(A=Time(), B=ScalarParameter() {defaultValue=1, parameterName=LightTexture_Speed})
  ) {speedY=0.3}
  n54 = Desaturation(
   None=LinearInterpolate(
    A=ScalarParameter() {parameterName=LightTextureLerp_A}
    Alpha=TextureSampleParameter2D(UVs=@n53) {parameterName=LightTexture, texture=&T_Swirly_Noise_highContrast}.RGB
   ) {constB=2}
   Fraction=Constant() {r=-1}
  )
  MaterialFunctionCall(Emissive=Multiply(A=Multiply(A=@n54, B=@n22.RGB), B=ScalarParameter() {defaultValue=1, parameterName=All_Light})) {materialFunction=&MF_FX_EmissiveAdjustments}
  n55 = ScalarParameter() {defaultValue=-2, parameterName=Disappear min}
  n56 = ScalarParameter() {defaultValue=1, parameterName=Disappear max}
  n57 = TextureCoordinate()
  n58 = VectorParameter() {parameterName=DissolveUV, defaultValue=<rgba 1.5 1 0 1>}
  n59 = AppendVector(A=@n58.R, B=@n58.R)
  n60 = TextureSampleParameter2D(UVs=Panner(Coordinate=Multiply(A=@n57, B=@n59), Speed=AppendVector(A=@n58.B, B=@n58.B)) {speedX=-0.7}) {
   parameterName=DissolveTexture
   samplerType=SAMPLERTYPE_LinearGrayscale
   texture=/Game/Global/Textures/Default/Caustics/T_Nosie_Caustics_Grayscale
  }
  n61 = ScalarParameter() {parameterName=Dissolve mask Min}
  n62 = ScalarParameter() {defaultValue=0.1, parameterName=Dissolve mask Max}
  n63 = ComponentMask(None=TextureCoordinate()) {g=true}
  n64 = Power(Base=Multiply(A=OneMinus(None=@n63)) {constB=3})
  n65 = Power(Base=@n60.R, Exp=Saturate(None=LinearInterpolate(A=@n61, B=@n62, Alpha=@n64) {constA=0.1, constB=2})) {constExponent=0.2}
  NamedRerouteDeclaration(None=Saturate(None=LinearInterpolate(A=@n55, B=@n56, Alpha=@n65) {constA=-2, constB=2})) {name=Dissolve, nodeColor=<rgba 0.0312 0 1 1>}
  Saturate(None=@n64)
 ]
}
