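# Translucent, unlit, two-sided Niagara mesh-particle material for the CornerWeek emote's trail effect.
# Two noise layers, one panning, combine with a clamped slash mask slid by a dynamic parameter, U-end fades and depth fade for opacity.
# Panning light textures tinted by particle color drive emissive, optionally modulating opacity.
# Switchable world position offset, off by default, adds SplineThicken width and local-axis bending weighted along U, both driven by dynamic parameters.
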Material /BRCosmetics/Effects/Fort_Effects/Emotes/CornerWeek/M_Emote_CornerWeek_Trail
graph {
 properties {
  bAutomaticallySetUsageInEditor=false
  bUsedWithNiagaraMeshParticles=true
  blendMode=BLEND_Translucent
  shadingModel=MSM_Unlit
  twoSided=true
  usedShadingModels=Unlit
 }
 nodes [
  n1 = DynamicParameter() {defaultValue=<rgba 0 1 1 1>, paramNames [ClampMask_AddX, SplineWidth, LightTexture_UVOffset, Param4]}
  n2 = Multiply(A=ScalarParameter() {defaultValue=3, parameterName=Spline_Base}, B=@n1.SplineWidth)
  n3 = Multiply(A=ScalarParameter() {defaultValue=3, parameterName=Spline_Tip}, B=@n1.SplineWidth)
  n4 = StaticBool()
  n5 = Transform(None=Constant3Vector() {constant=<rgba 1 0 0 0>}) {transformSourceType=TRANSFORMSOURCE_Local}
  n6 = OneMinus(None=ComponentMask(None=TextureCoordinate()) {r=true})
  n7 = DynamicParameter() {parameterIndex=1, paramNames [X WPO_Intensity, Y WPO_Intensity, Z WPO_Intensity, Param4]}
  n8 = Transform(None=Constant3Vector() {constant=<rgba 0 1 0 0>}) {transformSourceType=TRANSFORMSOURCE_Local}
  n9 = Add(A=Multiply(A=Multiply(A=@n5, B=@n6), B=@n7.X WPO_Intensity) {constB=10}, B=Multiply(A=@n7.Y WPO_Intensity, B=Multiply(A=@n8, B=@n6)))
  n10 = Constant3Vector() {constant=<rgba 0 0 1 0>}
  n11 = Multiply(A=@n7.Z WPO_Intensity, B=Multiply(A=Transform(None=@n10) {transformSourceType=TRANSFORMSOURCE_Local}, B=@n6))
  Add(
   A=LinearInterpolate(
    B=MaterialFunctionCall(WidthBase=@n2, WidthTip=@n3, Expand U or V UV Channel=@n4) {materialFunction=&SplineThicken}.WorldPositionOffset
    Alpha=ScalarParameter() {parameterName=Use_SplineThicken}
   )
   B=LinearInterpolate(B=Add(A=@n9, B=@n11), Alpha=ScalarParameter() {parameterName=Use_XYZ_WPO}) {constB=0}
  )
  n12 = ParticleColor()
  n13 = TextureSample() {samplerType=SAMPLERTYPE_LinearColor, texture=/Game/Global/Textures/Default/Tileable_Patterns/T_Tile_Smoke_Small}
  n14 = Constant() {r=1}
  n15 = DynamicParameter() {defaultValue=<rgba 0 1 1 1>, paramNames [ClampMask_AddX, SplineWidth, LightTexture_UVOffset, Param4]}
  n16 = Constant() {r=1}
  n17 = Add(A=TextureCoordinate(), B=AppendVector(A=Multiply(A=@n14, B=@n15.LightTexture_UVOffset), B=Multiply(A=@n16, B=@n15.LightTexture_UVOffset)))
  n18 = AppendVector(A=ScalarParameter() {defaultValue=1, parameterName=MaskTexture_U}, B=ScalarParameter() {defaultValue=1, parameterName=MAskTexture_V})
  n19 = TextureSampleParameter2D(UVs=Panner(Coordinate=Multiply(A=@n17, B=@n18)) {speedX=0.2, speedY=0.05}) {parameterName=MaskTexture, samplerType=SAMPLERTYPE_Grayscale, texture=/Game/Effects/Fort_Effects/Textures/Fire/T_WispyFireNoise_Soft_M}
  n20 = DynamicParameter() {defaultValue=<rgba 0 1 1 1>, paramNames [ClampMask_AddX, SplineWidth, LightTexture_UVOffset, Param4]}
  n21 = Constant()
  n22 = TextureSampleParameter2D(UVs=Add(A=TextureCoordinate(), B=AppendVector(A=@n20.ClampMask_AddX, B=@n21))) {
   parameterName=ClampTexture
   samplerSource=SSM_Clamp_WorldGroupSettings
   texture=/Game/Effects/Fort_Effects/GameplayPulginsAssets/Masamune/T_Masamune_Slash
  }
  n23 = Multiply(A=Multiply(A=@n12.A, B=Add(A=@n13.R, B=@n19.R)), B=Multiply(A=@n22.R))
  n24 = TextureCoordinate()
  n25 = AppendVector(A=ScalarParameter() {defaultValue=1, parameterName=LightTex02_U}, B=ScalarParameter() {defaultValue=1, parameterName=LightTex02_V})
  n26 = TextureSampleParameter2D(UVs=Panner(Coordinate=Multiply(A=@n24, B=@n25)) {speedX=0.3}) {parameterName=LightTex02, samplerType=SAMPLERTYPE_LinearGrayscale, texture=/Game/Global/Textures/Default/Caustics/T_Nosie_Caustics_Grayscale}
  n27 = TextureCoordinate()
  n28 = DynamicParameter() {defaultValue=<rgba 0 1 1 1>, paramNames [ClampMask_AddX, SplineWidth, LightTexture_UVOffset, Param4]}
  n29 = AppendVector(
   A=Multiply(A=ScalarParameter() {defaultValue=1, parameterName=LightTexture_UOffset}, B=@n28.LightTexture_UVOffset)
   B=Multiply(A=ScalarParameter() {defaultValue=1, parameterName=LightTexture_VOffset}, B=@n28.LightTexture_UVOffset)
  )
  n30 = ScalarParameter() {defaultValue=1, parameterName=LightTexture_U}
  n31 = ScalarParameter() {defaultValue=1, parameterName=LightTexture_V}
  n32 = Panner(
   Coordinate=Multiply(A=Add(A=@n27, B=@n29), B=AppendVector(A=@n30, B=@n31))
   Time=Multiply(A=Time(), B=ScalarParameter() {defaultValue=1, parameterName=LightTexture_Speed})
  ) {speedX=0.5}
  n33 = Add(
   A=Multiply(A=LinearInterpolate(Alpha=@n26.R), B=ScalarParameter() {defaultValue=1, parameterName=LightTex02_Mul})
   B=LinearInterpolate(
    A=ScalarParameter() {defaultValue=0.12, parameterName=LightTextureLerp_A}
    Alpha=TextureSampleParameter2D(UVs=@n32) {parameterName=LightTexture, texture=/Game/Weapons/WeaponSkins/Textures/T_Swirly_Noise_highContrast}.R
   ) {constB=0.6}
  )
  n34 = Multiply(A=Multiply(A=@n33, B=@n23), B=@n12.A)
  n35 = ScalarParameter() {parameterName=Use_MaskTexture}
  n36 = ComponentMask(None=TextureCoordinate()) {r=true}
  n37 = ScalarParameter() {defaultValue=8, parameterName=Top_Mask_Multiply}
  n38 = Saturate(None=Power(Base=Multiply(A=@n36, B=@n37) {constB=6}, Exp=ScalarParameter() {defaultValue=3, parameterName=Top_Mask_Power}) {constExponent=3})
  n39 = Multiply(A=OneMinus(None=@n36))
  n40 = ScalarParameter() {defaultValue=5, parameterName=B_Mask_Multiply}
  n41 = Saturate(None=Power(Base=Multiply(A=@n39, B=@n40) {constB=6}, Exp=ScalarParameter() {defaultValue=3, parameterName=B_Mask_Power}) {constExponent=3})
  DepthFade(
   Opacity=Multiply(A=LinearInterpolate(A=@n23, B=@n34, Alpha=@n35), B=Multiply(A=@n38, B=@n41))
   FadeDistance=ScalarParameter() {defaultValue=5, parameterName=DepthFade}
  )
  MaterialFunctionCall(Emissive=Multiply(A=@n33, B=@n12.RGB)) {materialFunction=&MF_FX_EmissiveAdjustments}
 ]
}
